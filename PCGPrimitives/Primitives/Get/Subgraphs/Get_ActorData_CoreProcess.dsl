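; PCG subgraph that gathers data from every world actor matching the ActorTag string parameter, serving as the core of an actor-data getter primitive.
; Get Actor Data uses that parameter as its selection tag, taking PCG component output or else parsing components, and always requeries actors.
; Delete Tags then strips tags matching the substring "input" before the data reaches Out.
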(graph Get_ActorData_CoreProcess
  (params (ActorTag :type String))
  (bind GetActorData_1 (Get%20Actor%20Data :ActorSelector (struct :ActorFilter AllWorldActors :ActorSelectionTag Districts :bSelectMultiple true) :bAlwaysRequeryActors true :bComponentsMustOverlapSelf false :bIgnorePCGGeneratedComponents false :bTrackActorsOnlyWithinBounds false :Mode GetDataFromPCGComponentOrParseComponents :ActorSelectionTag ActorTag))
  (bind DeleteTags_4 (Delete%20Tags :Operator Substring :SelectedTags "input"))
  (connect GetActorData_1 Out DeleteTags_4 In)
  (output Out DeleteTags_4))
